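# UI Modulate material, M_UI_Preset_Overlay, that darkens what is behind it within a rounded-corner shape.
# A linear gradient with adjustable start, end and angle, eased by EaseOutSine, blends from the Darkness parameter toward white.
# A max with the inverted rounded mask forces white outside the shape, leaving the background unchanged there.
# Vertex alpha lerps the result from white, so widget opacity fades the effect despite the multiply blend.
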Material /FortUILibrary/Systems/Items/MaterialInstances/M_UI_Preset_Overlay
graph {
 properties {bUsedWithStaticMesh=false, blendMode=BLEND_Modulate, materialDomain=MD_UI, pixelDepthOffsetMode=PDOM_Legacy, usedShadingModels=Default Lit}
 nodes [
  n1 = MaterialFunctionCall(Corner Radius=Constant() {r=20}, Radius in Pixels=Constant() {r=1}) {materialFunction=&MF_UI_CoreUI_RoundedMask}
  n2 = ScalarParameter() {parameterName=Darkness}
  n3 = ScalarParameter() {group=DefaultBackground, parameterName=BgGradientStart, sortPriority=10}
  n4 = ScalarParameter() {defaultValue=0.9, group=DefaultBackground, parameterName=BgGradientEnd, sortPriority=11}
  n5 = ScalarParameter() {
   defaultValue=0.25
   desc="Angle for linear gradient:\r\n0: from left to right\r\n0.25: from bottom to top\r\n0.5: from right to left\r\n0.75: from top to bottom"
   group=DefaultBackground
   parameterName=BgGradientAngle
   sortPriority=12
  }
  n6 = MaterialFunctionCall("Value (0-1)"=ComponentMask(None=MaterialFunctionCall(GradientStart=@n3, GradientEnd=@n4, RotationAngle=@n5) {materialFunction=&MF_UI_LinearGradient}.Result) {g=true}) {materialFunction=&MF_UI_EaseCurves}
  LinearInterpolate(B=Max(A=OneMinus(None=@n1.Shape Mask), B=LinearInterpolate(A=@n2, Alpha=@n6.EaseOutSine)), Alpha=VertexColor().A) {constA=1}
 ]
}
